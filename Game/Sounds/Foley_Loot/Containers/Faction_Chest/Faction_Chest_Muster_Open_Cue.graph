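# Sound cue for opening the Muster faction chest, a loot container.
# A mixer layers a latch and two uses of the same shimmer sound at slightly different levels, each through a pitch modulator with fixed volume, the second shimmer pitched down for a lower layer.
# It uses item-interact attenuation with occlusion enabled in its overrides, a loot-reward sound class, and overridden concurrency that prevents new instances at the limit.
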SoundCue /Game/Sounds/Foley_Loot/Containers/Faction_Chest/Faction_Chest_Muster_Open_Cue
root=Mixer_1
nodes [
 {id=Mixer_1, children [Modulator_1,Modulator_2,Modulator_4], properties {inputVolume [1,1.1,0.9]}}
 {id=Modulator_1, children [WavePlayer_3], properties {pitchMax=1.1, volumeMax=1, volumeMin=1}}
 {id=Modulator_2, children [WavePlayer_4], properties {pitchMax=1.1, volumeMax=1, volumeMin=1}}
 {id=Modulator_4, children [WavePlayer_5], properties {pitchMax=0.85, pitchMin=0.75, volumeMax=1, volumeMin=1}}
 {id=WavePlayer_3, properties {soundWaveAssetPtr=&/Faction_Chest_Muster_Open_Latch_01}}
 {id=WavePlayer_4, properties {soundWaveAssetPtr=&/Faction_Chest_Muster_Open_Shimmer_01}}
 {id=WavePlayer_5, properties {soundWaveAssetPtr=&Faction_Chest_Muster_Open_Shimmer_01}}
]
properties {
 attenuationSettings=/Game/Sounds/Attenuation/Fort_Item_Interact_Att
 bOverrideConcurrency=true
 maxDistance=1850
 soundClassObject=&Fort_Loot_Reward_PSM
 volumeMultiplier=1.5
 attenuationOverrides {bEnableOcclusion=true, falloffDistance=1000, occlusionTraceChannel=ECC_GameTraceChannel11, attenuationShapeExtents {x=500}}
 concurrencyOverrides {resolutionRule=PreventNew}
}
