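# Sound cue for the end of the Muster searching sound on the Faction Chest loot container, currently empty with no root node or sound nodes, so it plays no audio.
# It still carries playback settings: the Fort_Loot_Reward_PSM sound class and the shared Fort_Chests_Att attenuation asset.
# Stored attenuation overrides define distance falloff and trace-based occlusion with low-pass filtering and volume reduction.
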SoundCue /Game/Sounds/Foley_Loot/Containers/Faction_Chest/Faction_Chest_Muster_Searching_End_Cue
root=null
nodes []
properties {
 attenuationSettings=/Game/Sounds/Attenuation/Fort_Chests_Att
 maxDistance=1300
 soundClassObject=&Fort_Loot_Reward_PSM
 volumeMultiplier=1
 attenuationOverrides {
  bEnableOcclusion=true
  falloffDistance=1000
  occlusionInterpolationTime=0.3
  occlusionLowPassFilterFrequency=3000
  occlusionTraceChannel=ECC_GameTraceChannel11
  occlusionVolumeAttenuation=0.75
  attenuationShapeExtents {x=500}
 }
}
