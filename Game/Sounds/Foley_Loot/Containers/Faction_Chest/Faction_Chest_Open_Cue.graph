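# Sound cue for opening a faction chest, using shared item-interaction attenuation and a concurrency override that prevents new instances at the limit.
# The first two quality levels mix a latch-open sound, an eye-scan success sound and a random pick of three pitch-varied treasure chest sweeteners.
# The other two quality levels play only a single pitch-varied sweetener, an apparently optimized variant, as a cheaper version.
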SoundCue /Game/Sounds/Foley_Loot/Containers/Faction_Chest/Faction_Chest_Open_Cue
root=QualityLevel_0
nodes [
 {id=Mixer_0, children [Modulator_3,WavePlayer_0,WavePlayer_2], properties {inputVolume [0.5,0.75,1.25]}}
 {id=Modulator_0, children [WavePlayer_1], properties {pitchMax=1.5, pitchMin=1, volumeMax=0.7, volumeMin=0.7}}
 {id=Modulator_3, children [Random_6], properties {pitchMax=1.5, pitchMin=1, volumeMax=0.7, volumeMin=0.7}}
 {id=QualityLevel_0, children [Mixer_0,Mixer_0,Modulator_0,Modulator_0]}
 {id=Random_6, children [WavePlayer_17, WavePlayer_18, WavePlayer_19], properties {weights [1,1,1]}}
 {id=WavePlayer_0, properties {soundWaveAssetPtr=&/Faction_Chest_Latch_Open}}
 {id=WavePlayer_1, properties {soundWaveAssetPtr=/Game/Sounds/Foley_Loot/Containers/Treasure_Chest/treasure_chest_sweet_shorter_02_opt_02}}
 {id=WavePlayer_17, properties {soundWaveAssetPtr=/Game/Sounds/Foley_Loot/Containers/Treasure_Chest/treasure_chest_sweet_shorter_02}}
 {id=WavePlayer_18, properties {soundWaveAssetPtr=/Game/Sounds/Foley_Loot/Containers/Treasure_Chest/treasure_chest_sweet_shorter_03}}
 {id=WavePlayer_19, properties {soundWaveAssetPtr=/Game/Sounds/Foley_Loot/Containers/Treasure_Chest/treasure_chest_sweet_shorter_04}}
 {id=WavePlayer_2, properties {soundWaveAssetPtr=&/Faction_Chest_Eye_Scan_Success_01}}
]
properties {
 attenuationSettings=/Game/Sounds/Attenuation/Fort_Item_Interact_Att
 bOverrideConcurrency=true
 maxDistance=1850
 soundClassObject=&Fort_Loot_Reward_PSM
 attenuationOverrides {bEnableOcclusion=true, falloffDistance=1000, occlusionTraceChannel=ECC_GameTraceChannel11, attenuationShapeExtents {x=500}}
 concurrencyOverrides {resolutionRule=PreventNew}
}
