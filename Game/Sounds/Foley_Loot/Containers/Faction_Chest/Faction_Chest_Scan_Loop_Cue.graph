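# Sound cue for the faction chest scan loop, a loot container foley effect.
# A Mixer plays two waves together at equal volume, an eye scan and a general chest scan, with boosted cue volume.
# It is routed to the loot reward sound class and references the shared chest attenuation asset.
# Stored attenuation overrides define occlusion, low-pass and listener focus, but the override switch is not shown as enabled.
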SoundCue /Game/Sounds/Foley_Loot/Containers/Faction_Chest/Faction_Chest_Scan_Loop_Cue
root=Mixer_0
nodes [
 {id=Mixer_0, children [WavePlayer_0,WavePlayer_2], properties {inputVolume [1,1]}}
 {id=WavePlayer_0, properties {soundWaveAssetPtr=&/Faction_Chest_Eye_Scan_01}}
 {id=WavePlayer_2, properties {soundWaveAssetPtr=&/Faction_Chest_Scan}}
]
properties {
 attenuationSettings=/Game/Sounds/Attenuation/Fort_Chests_Att
 maxDistance=1300
 soundClassObject=&Fort_Loot_Reward_PSM
 volumeMultiplier=1.5
 attenuationOverrides {
  bAttenuateWithLPF=true
  bEnableFocusInterpolation=true
  bEnableListenerFocus=true
  bEnableOcclusion=true
  distanceAlgorithm=Logarithmic
  falloffDistance=1000
  focusAttackInterpSpeed=2
  focusReleaseInterpSpeed=2
  lPFFrequencyAtMax=400
  lPFRadiusMax=350
  lPFRadiusMin=150
  nonFocusVolumeAttenuation=0.5
  occlusionInterpolationTime=0.3
  occlusionLowPassFilterFrequency=3000
  occlusionTraceChannel=ECC_GameTraceChannel11
  occlusionVolumeAttenuation=0.75
  attenuationShapeExtents {x=500}
 }
}
